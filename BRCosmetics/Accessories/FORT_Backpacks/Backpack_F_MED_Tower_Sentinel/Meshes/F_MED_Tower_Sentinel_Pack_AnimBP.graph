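# Animation Blueprint for the Tower Sentinel backpack cosmetic, posing its fingers and simulating its capes.
# The input pose gets additive backpack and finger offsets, then a jump-weighted cape lift and two trail controllers.
# Body and floorCOL bones are then overwritten with world-space transforms from variables not set here, feeding a final RigidBody simulation with variable gravity.
# A skydiving smoke_root and RigidBody branch is disconnected from the output; the event graph is empty.
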AnimBlueprint /BRCosmetics/Accessories/FORT_Backpacks/Backpack_F_MED_Tower_Sentinel/Meshes/F_MED_Tower_Sentinel_Pack_AnimBP
target_skeleton=&/F_MED_Tower_Sentinel_Pack_Skeleton
graphs [
 AnimGraph {nodes [
   n1 = "Transform (Modify) Bone - Bone: smoke_root"!(ComponentPose, bAlphaBoolEnabled:bool=Get bIsSkyDiving(), Translation="0,0,0", Rotation="0,-50,0", Scale="1,1,1") -> (Pose) {
    alphaInputType=Bool
    rotationMode=BMM_Additive
    rotationSpace=BCS_BoneSpace
    scaleMode=BMM_Additive
    scaleSpace=BCS_BoneSpace
    translationMode=BMM_Additive
    translationSpace=BCS_BoneSpace
    alphaBoolBlend {blendOption=Linear}
    boneToModify {boneName=smoke_root}
    rotation {yaw=-50}
   }
   n2 = RigidBody(ComponentPose=@n1) {
    bClampLinearTranslationLimitToRefPose=true
    bOverrideWorldGravity=true
    simulationSpace=BaseBoneSpace
    alphaBoolBlend {blendOption=Linear}
    baseBoneRef {boneName=root}
    componentLinearAccScale=<0.3 0.3 0.3>
    componentLinearVelScale=<0.3 0.3 0.3>
    overrideWorldGravity {z=-300}
   }
   Component To Local(ComponentPose=@n2)
   Copy Pose From Mesh()
   Local Space Ref Pose()
   n3 = "Transform (Modify) Bone - Bone: body_root"(
    ComponentPose=Local To Component(LocalPose=Use cached pose 'trailControlCache'())
    Alpha=Get IsEmotingAlpha() -> (IsEmotingAlpha:float)
    Translation=Get BodyRootTran() -> (BodyRootTran:Vector)
    Rotation=Get BodyRootRot() -> (BodyRootRot:Rotator)
    Scale="1,1,1"
   ) {
    rotationMode=BMM_Replace
    rotationSpace=BCS_WorldSpace
    scaleSpace=BCS_BoneSpace
    translationMode=BMM_Replace
    translationSpace=BCS_WorldSpace
    alphaBoolBlend {blendOption=Linear}
    alphaScaleBiasClamp {inRange {max=0, min=-40}, outRange {max=0, min=1}}
    boneToModify {boneName=body_root}
   }
   n4 = "Transform (Modify) Bone - Bone: pelvis"(
    ComponentPose=@n3
    Alpha=1
    Translation=Get PelvisTran() -> (PelvisTran:Vector)
    Rotation=Get PelvisRot() -> (PelvisRot:Rotator)
    Scale="1,1,1"
   ) {
    rotationMode=BMM_Replace
    rotationSpace=BCS_WorldSpace
    scaleSpace=BCS_BoneSpace
    translationMode=BMM_Replace
    translationSpace=BCS_WorldSpace
    alphaBoolBlend {blendOption=Linear}
    alphaScaleBiasClamp {inRange {max=0, min=-40}, outRange {max=0, min=1}}
    boneToModify {boneName=pelvis}
   }
   n5 = "Transform (Modify) Bone - Bone: floorCOL"(ComponentPose=@n4, Alpha=1, Translation=Get RootTran() -> (RootTran:Vector), Rotation=Get RootRot() -> (RootRot:Rotator), Scale="1,1,1") {
    rotationMode=BMM_Replace
    rotationSpace=BCS_WorldSpace
    scaleSpace=BCS_BoneSpace
    translationMode=BMM_Replace
    translationSpace=BCS_WorldSpace
    alphaBoolBlend {blendOption=Linear}
    alphaScaleBiasClamp {inRange {max=0, min=-40}, outRange {max=0, min=1}}
    boneToModify {boneName=floorCOL}
   }
   n6 = "Transform (Modify) Bone - Bone: thigh_l"(
    ComponentPose=@n5
    Alpha=1
    Translation=Get ThighLTran() -> (ThighLTran:Vector)
    Rotation=Get ThighLRot() -> (ThighLRot:Rotator)
    Scale="1,1,1"
   ) {
    rotationMode=BMM_Replace
    rotationSpace=BCS_WorldSpace
    scaleSpace=BCS_BoneSpace
    translationMode=BMM_Replace
    translationSpace=BCS_WorldSpace
    alphaBoolBlend {blendOption=Linear}
    alphaScaleBiasClamp {inRange {max=0, min=-40}, outRange {max=0, min=1}}
    boneToModify {boneName=thigh_l}
   }
   n7 = "Transform (Modify) Bone - Bone: thigh_r"(
    ComponentPose=@n6
    Alpha=1
    Translation=Get ThighRTran() -> (ThighRTran:Vector)
    Rotation=Get ThighRRot() -> (ThighRRot:Rotator)
    Scale="1,1,1"
   ) {
    rotationMode=BMM_Replace
    rotationSpace=BCS_WorldSpace
    scaleSpace=BCS_BoneSpace
    translationMode=BMM_Replace
    translationSpace=BCS_WorldSpace
    alphaBoolBlend {blendOption=Linear}
    alphaScaleBiasClamp {inRange {max=0, min=-40}, outRange {max=0, min=1}}
    boneToModify {boneName=thigh_r}
   }
   n8 = "Transform (Modify) Bone - Bone: calf_l"(ComponentPose=@n7, Alpha=1, Translation=Get CalfLTran() -> (CalfLTran:Vector), Rotation=Get CalfLRot() -> (CalfLRot:Rotator), Scale="1,1,1") {
    rotationMode=BMM_Replace
    rotationSpace=BCS_WorldSpace
    scaleSpace=BCS_BoneSpace
    translationMode=BMM_Replace
    translationSpace=BCS_WorldSpace
    alphaBoolBlend {blendOption=Linear}
    alphaScaleBiasClamp {inRange {max=0, min=-40}, outRange {max=0, min=1}}
    boneToModify {boneName=calf_l}
   }
   n9 = "Transform (Modify) Bone - Bone: calf_r"(ComponentPose=@n8, Alpha=1, Translation=Get CalfRTran() -> (CalfRTran:Vector), Rotation=Get CalfRRot() -> (CalfRRot:Rotator), Scale="1,1,1") {
    rotationMode=BMM_Replace
    rotationSpace=BCS_WorldSpace
    scaleSpace=BCS_BoneSpace
    translationMode=BMM_Replace
    translationSpace=BCS_WorldSpace
    alphaBoolBlend {blendOption=Linear}
    alphaScaleBiasClamp {inRange {max=0, min=-40}, outRange {max=0, min=1}}
    boneToModify {boneName=calf_r}
   }
   n10 = "Transform (Modify) Bone - Bone: spine_01"(
    ComponentPose=@n9
    Alpha=1
    Translation=Get Spine01Tran() -> (Spine01Tran:Vector)
    Rotation=Get Spine01Rot() -> (Spine01Rot:Rotator)
    Scale="1,1,1"
   ) {
    rotationMode=BMM_Replace
    rotationSpace=BCS_WorldSpace
    scaleSpace=BCS_BoneSpace
    translationMode=BMM_Replace
    translationSpace=BCS_WorldSpace
    alphaBoolBlend {blendOption=Linear}
    alphaScaleBiasClamp {inRange {max=0, min=-40}, outRange {max=0, min=1}}
    boneToModify {boneName=spine_01}
   }
   n11 = "Transform (Modify) Bone - Bone: spine_02"(
    ComponentPose=@n10
    Alpha=1
    Translation=Get Spine02Tran() -> (Spine02Tran:Vector)
    Rotation=Get Spine02Rot() -> (Spine02Rot:Rotator)
    Scale="1,1,1"
   ) {
    rotationMode=BMM_Replace
    rotationSpace=BCS_WorldSpace
    scaleSpace=BCS_BoneSpace
    translationMode=BMM_Replace
    translationSpace=BCS_WorldSpace
    alphaBoolBlend {blendOption=Linear}
    alphaScaleBiasClamp {inRange {max=0, min=-40}, outRange {max=0, min=1}}
    boneToModify {boneName=spine_02}
   }
   n12 = "Transform (Modify) Bone - Bone: spine_03"(
    ComponentPose=@n11
    Alpha=1
    Translation=Get Spine03Tran() -> (Spine03Tran:Vector)
    Rotation=Get Spine03Rot() -> (Spine03Rot:Rotator)
    Scale="1,1,1"
   ) {
    rotationMode=BMM_Replace
    rotationSpace=BCS_WorldSpace
    scaleSpace=BCS_BoneSpace
    translationMode=BMM_Replace
    translationSpace=BCS_WorldSpace
    alphaBoolBlend {blendOption=Linear}
    alphaScaleBiasClamp {inRange {max=0, min=-40}, outRange {max=0, min=1}}
    boneToModify {boneName=spine_03}
   }
   n13 = "Transform (Modify) Bone - Bone: spine_04"(
    ComponentPose=@n12
    Alpha=1
    Translation=Get Spine04Tran() -> (Spine04Tran:Vector)
    Rotation=Get Spine04Rot() -> (Spine04Rot:Rotator)
    Scale="1,1,1"
   ) {
    rotationMode=BMM_Replace
    rotationSpace=BCS_WorldSpace
    scaleSpace=BCS_BoneSpace
    translationMode=BMM_Replace
    translationSpace=BCS_WorldSpace
    alphaBoolBlend {blendOption=Linear}
    alphaScaleBiasClamp {inRange {max=0, min=-40}, outRange {max=0, min=1}}
    boneToModify {boneName=spine_04}
   }
   n14 = "Transform (Modify) Bone - Bone: spine_05"(
    ComponentPose=@n13
    Alpha=1
    Translation=Get Spine05Tran() -> (Spine05Tran:Vector)
    Rotation=Get Spine05Rot() -> (Spine05Rot:Rotator)
    Scale="1,1,1"
   ) {
    rotationMode=BMM_Replace
    rotationSpace=BCS_WorldSpace
    scaleSpace=BCS_BoneSpace
    translationMode=BMM_Replace
    translationSpace=BCS_WorldSpace
    alphaBoolBlend {blendOption=Linear}
    alphaScaleBiasClamp {inRange {max=0, min=-40}, outRange {max=0, min=1}}
    boneToModify {boneName=spine_05}
   }
   n15 = "Transform (Modify) Bone - Bone: clavicle_l"(
    ComponentPose=@n14
    Alpha=1
    Translation=Get ClavicleLTran() -> (ClavicleLTran:Vector)
    Rotation=Get ClavicleLRot() -> (ClavicleLRot:Rotator)
    Scale="1,1,1"
   ) {
    rotationMode=BMM_Replace
    rotationSpace=BCS_WorldSpace
    scaleSpace=BCS_BoneSpace
    translationMode=BMM_Replace
    translationSpace=BCS_WorldSpace
    alphaBoolBlend {blendOption=Linear}
    alphaScaleBiasClamp {inRange {max=0, min=-40}, outRange {max=0, min=1}}
    boneToModify {boneName=clavicle_l}
   }
   n16 = "Transform (Modify) Bone - Bone: clavicle_r"(
    ComponentPose=@n15
    Alpha=1
    Translation=Get ClavicleRTran() -> (ClavicleRTran:Vector)
    Rotation=Get ClavicleRRot() -> (ClavicleRRot:Rotator)
    Scale="1,1,1"
   ) {
    rotationMode=BMM_Replace
    rotationSpace=BCS_WorldSpace
    scaleSpace=BCS_BoneSpace
    translationMode=BMM_Replace
    translationSpace=BCS_WorldSpace
    alphaBoolBlend {blendOption=Linear}
    alphaScaleBiasClamp {inRange {max=0, min=-40}, outRange {max=0, min=1}}
    boneToModify {boneName=clavicle_r}
   }
   n17 = "Transform (Modify) Bone - Bone: upperarm_l"(
    ComponentPose=@n16
    Alpha=1
    Translation=Get UpperarmLTran() -> (UpperarmLTran:Vector)
    Rotation=Get UpperarmLRot() -> (UpperarmLRot:Rotator)
    Scale="1,1,1"
   ) {
    rotationMode=BMM_Replace
    rotationSpace=BCS_WorldSpace
    scaleSpace=BCS_BoneSpace
    translationMode=BMM_Replace
    translationSpace=BCS_WorldSpace
    alphaBoolBlend {blendOption=Linear}
    alphaScaleBiasClamp {inRange {max=0, min=-40}, outRange {max=0, min=1}}
    boneToModify {boneName=upperarm_l}
   }
   n18 = "Transform (Modify) Bone - Bone: upperarm_r"(
    ComponentPose=@n17
    Alpha=1
    Translation=Get UpperarmRTran() -> (UpperarmRTran:Vector)
    Rotation=Get UpperarmRRot() -> (UpperarmRRot:Rotator)
    Scale="1,1,1"
   ) {
    rotationMode=BMM_Replace
    rotationSpace=BCS_WorldSpace
    scaleSpace=BCS_BoneSpace
    translationMode=BMM_Replace
    translationSpace=BCS_WorldSpace
    alphaBoolBlend {blendOption=Linear}
    alphaScaleBiasClamp {inRange {max=0, min=-40}, outRange {max=0, min=1}}
    boneToModify {boneName=upperarm_r}
   }
   n19 = "Transform (Modify) Bone - Bone: lowerarm_l"(
    ComponentPose=@n18
    Alpha=1
    Translation=Get LowerarmLTran() -> (LowerarmLTran:Vector)
    Rotation=Get LowerarmLRot() -> (LowerarmLRot:Rotator)
    Scale="1,1,1"
   ) {
    rotationMode=BMM_Replace
    rotationSpace=BCS_WorldSpace
    scaleSpace=BCS_BoneSpace
    translationMode=BMM_Replace
    translationSpace=BCS_WorldSpace
    alphaBoolBlend {blendOption=Linear}
    alphaScaleBiasClamp {inRange {max=0, min=-40}, outRange {max=0, min=1}}
    boneToModify {boneName=lowerarm_l}
   }
   n20 = "Transform (Modify) Bone - Bone: lowerarm_r"(
    ComponentPose=@n19
    Alpha=1
    Translation=Get LowerarmRTran() -> (LowerarmRTran:Vector)
    Rotation=Get LowerarmRRot() -> (LowerarmRRot:Rotator)
    Scale="1,1,1"
   ) {
    rotationMode=BMM_Replace
    rotationSpace=BCS_WorldSpace
    scaleSpace=BCS_BoneSpace
    translationMode=BMM_Replace
    translationSpace=BCS_WorldSpace
    alphaBoolBlend {blendOption=Linear}
    alphaScaleBiasClamp {inRange {max=0, min=-40}, outRange {max=0, min=1}}
    boneToModify {boneName=lowerarm_r}
   }
   n21 = RigidBody!(
    ComponentPose=@n20
    Alpha
    OverrideWorldGravity:Vector=Get FinalGravity() -> (FinalGravity:Vector)
    ExternalForce
    ComponentLinearAccScale:Vector=Get CompLinearAccScale() -> (CompLinearAccScale:Vector)
    ComponentLinearVelScale:Vector=Get CompLinearVelScale() -> (CompLinearVelScale:Vector)
   ) -> (Pose) {
    bClampLinearTranslationLimitToRefPose=true
    bOverrideWorldGravity=true
    overridePhysicsAsset=&/F_MED_Tower_Sentinel_Pack_Physics1
    simulationSpace=BaseBoneSpace
    alphaBoolBlend {blendOption=Linear}
    baseBoneRef {boneName=body_root}
    componentAppliedLinearAccClamp=<5000 5000 5000>
    componentLinearAccScale=<0.1 0.1 0.1>
    componentLinearVelScale=<0.1 0.1 0.1>
    overrideWorldGravity {z=-980}
    simSpaceSettings {
     maxAngularAcceleration=100
     maxAngularVelocity=100
     maxLinearAcceleration=100
     maxLinearVelocity=100
     velocityScaleZ=0.5
     externalLinearDragV=<10 10 10>
    }
   }
   Output Pose(Result=Component To Local(ComponentPose=@n21)) {layerGroup=None}
   n22 = "Transform (Modify) Bone - Bone: backpack_root"(ComponentPose=Local To Component(LocalPose=Input Pose()), Alpha=1, Translation="0,5,0", Rotation="0,0,20", Scale="1,1,1") {
    rotationMode=BMM_Additive
    rotationSpace=BCS_BoneSpace
    translationMode=BMM_Additive
    translationSpace=BCS_BoneSpace
    alphaBoolBlend {blendOption=Linear}
    alphaScaleBiasClamp {bInterpResult=true}
    boneToModify {boneName=backpack_root}
    rotation {roll=20}
    translation {y=5}
   }
   n23 = "Transform (Modify) Bone - Bone: index_a_l"!(ComponentPose=@n22, bAlphaBoolEnabled:bool=True, Translation="0,0,0", Rotation=Get IndexA() -> (IndexA:Rotator), Scale="1,1,1") -> (Pose) {
    alphaInputType=Bool
    rotationMode=BMM_Additive
    rotationSpace=BCS_BoneSpace
    scaleMode=BMM_Additive
    scaleSpace=BCS_BoneSpace
    translationMode=BMM_Additive
    translationSpace=BCS_BoneSpace
    alphaBoolBlend {blendOption=Linear}
    boneToModify {boneName=index_a_l}
   }
   n24 = "Transform (Modify) Bone - Bone: index_b_l"!(ComponentPose=@n23, bAlphaBoolEnabled:bool=True, Translation="0,0,0", Rotation=Get IndexB() -> (IndexB:Rotator), Scale="1,1,1") -> (Pose) {
    alphaInputType=Bool
    rotationMode=BMM_Additive
    rotationSpace=BCS_BoneSpace
    scaleMode=BMM_Additive
    scaleSpace=BCS_BoneSpace
    translationMode=BMM_Additive
    translationSpace=BCS_BoneSpace
    alphaBoolBlend {blendOption=Linear}
    boneToModify {boneName=index_b_l}
   }
   n25 = "Transform (Modify) Bone - Bone: index_c_l"!(ComponentPose=@n24, bAlphaBoolEnabled:bool=True, Translation="0,0,0", Rotation=Get IndexC() -> (IndexC:Rotator), Scale="1,1,1") -> (Pose) {
    alphaInputType=Bool
    rotationMode=BMM_Additive
    rotationSpace=BCS_BoneSpace
    scaleMode=BMM_Additive
    scaleSpace=BCS_BoneSpace
    translationMode=BMM_Additive
    translationSpace=BCS_BoneSpace
    alphaBoolBlend {blendOption=Linear}
    boneToModify {boneName=index_c_l}
   }
   n26 = "Transform (Modify) Bone - Bone: middle_a_l"!(ComponentPose=@n25, bAlphaBoolEnabled:bool=True, Translation="0,0,0", Rotation=Get MiddleA() -> (MiddleA:Rotator), Scale="1,1,1") -> (Pose) {
    alphaInputType=Bool
    rotationMode=BMM_Additive
    rotationSpace=BCS_BoneSpace
    scaleMode=BMM_Additive
    scaleSpace=BCS_BoneSpace
    translationMode=BMM_Additive
    translationSpace=BCS_BoneSpace
    alphaBoolBlend {blendOption=Linear}
    boneToModify {boneName=middle_a_l}
   }
   n27 = "Transform (Modify) Bone - Bone: middle_b_l"!(ComponentPose=@n26, bAlphaBoolEnabled:bool=True, Translation="0,0,0", Rotation=Get MiddleB() -> (MiddleB:Rotator), Scale="1,1,1") -> (Pose) {
    alphaInputType=Bool
    rotationMode=BMM_Additive
    rotationSpace=BCS_BoneSpace
    scaleMode=BMM_Additive
    scaleSpace=BCS_BoneSpace
    translationMode=BMM_Additive
    translationSpace=BCS_BoneSpace
    alphaBoolBlend {blendOption=Linear}
    boneToModify {boneName=middle_b_l}
   }
   n28 = "Transform (Modify) Bone - Bone: middle_c_l"!(ComponentPose=@n27, bAlphaBoolEnabled:bool=True, Translation="0,0,0", Rotation=Get MiddleC() -> (MiddleC:Rotator), Scale="1,1,1") -> (Pose) {
    alphaInputType=Bool
    rotationMode=BMM_Additive
    rotationSpace=BCS_BoneSpace
    scaleMode=BMM_Additive
    scaleSpace=BCS_BoneSpace
    translationMode=BMM_Additive
    translationSpace=BCS_BoneSpace
    alphaBoolBlend {blendOption=Linear}
    boneToModify {boneName=middle_c_l}
   }
   n29 = "Transform (Modify) Bone - Bone: ring_a_l"!(ComponentPose=@n28, bAlphaBoolEnabled:bool=True, Translation="0,0,0", Rotation=Get RingA() -> (RingA:Rotator), Scale="1,1,1") -> (Pose) {
    alphaInputType=Bool
    rotationMode=BMM_Additive
    rotationSpace=BCS_BoneSpace
    scaleMode=BMM_Additive
    scaleSpace=BCS_BoneSpace
    translationMode=BMM_Additive
    translationSpace=BCS_BoneSpace
    alphaBoolBlend {blendOption=Linear}
    boneToModify {boneName=ring_a_l}
   }
   n30 = "Transform (Modify) Bone - Bone: ring_b_l"!(ComponentPose=@n29, bAlphaBoolEnabled:bool=True, Translation="0,0,0", Rotation=Get RingB() -> (RingB:Rotator), Scale="1,1,1") -> (Pose) {
    alphaInputType=Bool
    rotationMode=BMM_Additive
    rotationSpace=BCS_BoneSpace
    scaleMode=BMM_Additive
    scaleSpace=BCS_BoneSpace
    translationMode=BMM_Additive
    translationSpace=BCS_BoneSpace
    alphaBoolBlend {blendOption=Linear}
    boneToModify {boneName=ring_b_l}
   }
   n31 = "Transform (Modify) Bone - Bone: ring_c_l"!(ComponentPose=@n30, bAlphaBoolEnabled:bool=True, Translation="0,0,0", Rotation=Get RingC() -> (RingC:Rotator), Scale="1,1,1") -> (Pose) {
    alphaInputType=Bool
    rotationMode=BMM_Additive
    rotationSpace=BCS_BoneSpace
    scaleMode=BMM_Additive
    scaleSpace=BCS_BoneSpace
    translationMode=BMM_Additive
    translationSpace=BCS_BoneSpace
    alphaBoolBlend {blendOption=Linear}
    boneToModify {boneName=ring_c_l}
   }
   n32 = "Transform (Modify) Bone - Bone: pinky_a_l"!(ComponentPose=@n31, bAlphaBoolEnabled:bool=True, Translation="0,0,0", Rotation=Get PinkyA() -> (PinkyA:Rotator), Scale="1,1,1") -> (Pose) {
    alphaInputType=Bool
    rotationMode=BMM_Additive
    rotationSpace=BCS_BoneSpace
    scaleMode=BMM_Additive
    scaleSpace=BCS_BoneSpace
    translationMode=BMM_Additive
    translationSpace=BCS_BoneSpace
    alphaBoolBlend {blendOption=Linear}
    boneToModify {boneName=pinky_a_l}
   }
   n33 = "Transform (Modify) Bone - Bone: pinky_b_l"!(ComponentPose=@n32, bAlphaBoolEnabled:bool=True, Translation="0,0,0", Rotation=Get PinkyB() -> (PinkyB:Rotator), Scale="1,1,1") -> (Pose) {
    alphaInputType=Bool
    rotationMode=BMM_Additive
    rotationSpace=BCS_BoneSpace
    scaleMode=BMM_Additive
    scaleSpace=BCS_BoneSpace
    translationMode=BMM_Additive
    translationSpace=BCS_BoneSpace
    alphaBoolBlend {blendOption=Linear}
    boneToModify {boneName=pinky_b_l}
   }
   n34 = "Transform (Modify) Bone - Bone: pinky_c_l"!(ComponentPose=@n33, bAlphaBoolEnabled:bool=True, Translation="0,0,0", Rotation=Get PinkyC() -> (PinkyC:Rotator), Scale="1,1,1") -> (Pose) {
    alphaInputType=Bool
    rotationMode=BMM_Additive
    rotationSpace=BCS_BoneSpace
    scaleMode=BMM_Additive
    scaleSpace=BCS_BoneSpace
    translationMode=BMM_Additive
    translationSpace=BCS_BoneSpace
    alphaBoolBlend {blendOption=Linear}
    boneToModify {boneName=pinky_c_l}
   }
   n35 = "Transform (Modify) Bone - Bone: index_a_r"!(ComponentPose=@n34, bAlphaBoolEnabled:bool=True, Translation="0,0,0", Rotation=Get IndexA(), Scale="1,1,1") -> (Pose) {
    alphaInputType=Bool
    rotationMode=BMM_Additive
    rotationSpace=BCS_BoneSpace
    scaleMode=BMM_Additive
    scaleSpace=BCS_BoneSpace
    translationMode=BMM_Additive
    translationSpace=BCS_BoneSpace
    alphaBoolBlend {blendOption=Linear}
    boneToModify {boneName=index_a_r}
   }
   n36 = "Transform (Modify) Bone - Bone: index_b_r"!(ComponentPose=@n35, bAlphaBoolEnabled:bool=True, Translation="0,0,0", Rotation=Get IndexB(), Scale="1,1,1") -> (Pose) {
    alphaInputType=Bool
    rotationMode=BMM_Additive
    rotationSpace=BCS_BoneSpace
    scaleMode=BMM_Additive
    scaleSpace=BCS_BoneSpace
    translationMode=BMM_Additive
    translationSpace=BCS_BoneSpace
    alphaBoolBlend {blendOption=Linear}
    boneToModify {boneName=index_b_r}
   }
   n37 = "Transform (Modify) Bone - Bone: index_c_r"!(ComponentPose=@n36, bAlphaBoolEnabled:bool=True, Translation="0,0,0", Rotation=Get IndexC(), Scale="1,1,1") -> (Pose) {
    alphaInputType=Bool
    rotationMode=BMM_Additive
    rotationSpace=BCS_BoneSpace
    scaleMode=BMM_Additive
    scaleSpace=BCS_BoneSpace
    translationMode=BMM_Additive
    translationSpace=BCS_BoneSpace
    alphaBoolBlend {blendOption=Linear}
    boneToModify {boneName=index_c_r}
   }
   n38 = "Transform (Modify) Bone - Bone: middle_a_r"!(ComponentPose=@n37, bAlphaBoolEnabled:bool=True, Translation="0,0,0", Rotation=Get MiddleA(), Scale="1,1,1") -> (Pose) {
    alphaInputType=Bool
    rotationMode=BMM_Additive
    rotationSpace=BCS_BoneSpace
    scaleMode=BMM_Additive
    scaleSpace=BCS_BoneSpace
    translationMode=BMM_Additive
    translationSpace=BCS_BoneSpace
    alphaBoolBlend {blendOption=Linear}
    boneToModify {boneName=middle_a_r}
   }
   n39 = "Transform (Modify) Bone - Bone: middle_b_r"!(ComponentPose=@n38, bAlphaBoolEnabled:bool=True, Translation="0,0,0", Rotation=Get MiddleB(), Scale="1,1,1") -> (Pose) {
    alphaInputType=Bool
    rotationMode=BMM_Additive
    rotationSpace=BCS_BoneSpace
    scaleMode=BMM_Additive
    scaleSpace=BCS_BoneSpace
    translationMode=BMM_Additive
    translationSpace=BCS_BoneSpace
    alphaBoolBlend {blendOption=Linear}
    boneToModify {boneName=middle_b_r}
   }
   n40 = "Transform (Modify) Bone - Bone: middle_c_r"!(ComponentPose=@n39, bAlphaBoolEnabled:bool=True, Translation="0,0,0", Rotation=Get MiddleC(), Scale="1,1,1") -> (Pose) {
    alphaInputType=Bool
    rotationMode=BMM_Additive
    rotationSpace=BCS_BoneSpace
    scaleMode=BMM_Additive
    scaleSpace=BCS_BoneSpace
    translationMode=BMM_Additive
    translationSpace=BCS_BoneSpace
    alphaBoolBlend {blendOption=Linear}
    boneToModify {boneName=middle_c_r}
   }
   n41 = "Transform (Modify) Bone - Bone: ring_a_r"!(ComponentPose=@n40, bAlphaBoolEnabled:bool=True, Translation="0,0,0", Rotation=Get RingA(), Scale="1,1,1") -> (Pose) {
    alphaInputType=Bool
    rotationMode=BMM_Additive
    rotationSpace=BCS_BoneSpace
    scaleMode=BMM_Additive
    scaleSpace=BCS_BoneSpace
    translationMode=BMM_Additive
    translationSpace=BCS_BoneSpace
    alphaBoolBlend {blendOption=Linear}
    boneToModify {boneName=ring_a_r}
   }
   n42 = "Transform (Modify) Bone - Bone: ring_b_r"!(ComponentPose=@n41, bAlphaBoolEnabled:bool=True, Translation="0,0,0", Rotation=Get RingB(), Scale="1,1,1") -> (Pose) {
    alphaInputType=Bool
    rotationMode=BMM_Additive
    rotationSpace=BCS_BoneSpace
    scaleMode=BMM_Additive
    scaleSpace=BCS_BoneSpace
    translationMode=BMM_Additive
    translationSpace=BCS_BoneSpace
    alphaBoolBlend {blendOption=Linear}
    boneToModify {boneName=ring_b_r}
   }
   n43 = "Transform (Modify) Bone - Bone: ring_c_r"!(ComponentPose=@n42, bAlphaBoolEnabled:bool=True, Translation="0,0,0", Rotation=Get RingC(), Scale="1,1,1") -> (Pose) {
    alphaInputType=Bool
    rotationMode=BMM_Additive
    rotationSpace=BCS_BoneSpace
    scaleMode=BMM_Additive
    scaleSpace=BCS_BoneSpace
    translationMode=BMM_Additive
    translationSpace=BCS_BoneSpace
    alphaBoolBlend {blendOption=Linear}
    boneToModify {boneName=ring_c_r}
   }
   n44 = "Transform (Modify) Bone - Bone: pinky_a_r"!(ComponentPose=@n43, bAlphaBoolEnabled:bool=True, Translation="0,0,0", Rotation=Get PinkyA(), Scale="1,1,1") -> (Pose) {
    alphaInputType=Bool
    rotationMode=BMM_Additive
    rotationSpace=BCS_BoneSpace
    scaleMode=BMM_Additive
    scaleSpace=BCS_BoneSpace
    translationMode=BMM_Additive
    translationSpace=BCS_BoneSpace
    alphaBoolBlend {blendOption=Linear}
    boneToModify {boneName=pinky_a_r}
   }
   n45 = "Transform (Modify) Bone - Bone: pinky_b_r"!(ComponentPose=@n44, bAlphaBoolEnabled:bool=True, Translation="0,0,0", Rotation=Get PinkyB(), Scale="1,1,1") -> (Pose) {
    alphaInputType=Bool
    rotationMode=BMM_Additive
    rotationSpace=BCS_BoneSpace
    scaleMode=BMM_Additive
    scaleSpace=BCS_BoneSpace
    translationMode=BMM_Additive
    translationSpace=BCS_BoneSpace
    alphaBoolBlend {blendOption=Linear}
    boneToModify {boneName=pinky_b_r}
   }
   n46 = "Transform (Modify) Bone - Bone: pinky_c_r"!(ComponentPose=@n45, bAlphaBoolEnabled:bool=True, Translation="0,0,0", Rotation=Get PinkyC(), Scale="1,1,1") -> (Pose) {
    alphaInputType=Bool
    rotationMode=BMM_Additive
    rotationSpace=BCS_BoneSpace
    scaleMode=BMM_Additive
    scaleSpace=BCS_BoneSpace
    translationMode=BMM_Additive
    translationSpace=BCS_BoneSpace
    alphaBoolBlend {blendOption=Linear}
    boneToModify {boneName=pinky_c_r}
   }
   Save cached pose 'initCache'(Pose=Component To Local(ComponentPose=@n46)) {cacheName=initCache}
   n47 = Get JumpingAlpha() -> (JumpingAlpha:float)
   n48 = "Transform (Modify) Bone - Bone: dyn_cape_01_l_MSTR"(ComponentPose=Local To Component(LocalPose=Use cached pose 'initCache'()), Alpha=@n47, Translation="0,0,0", Rotation="40,0,0", Scale="1,1,1") {
    rotationMode=BMM_Additive
    rotationSpace=BCS_BoneSpace
    translationMode=BMM_Additive
    translationSpace=BCS_BoneSpace
    alphaBoolBlend {blendOption=Linear}
    alphaScaleBiasClamp {bInterpResult=true}
    boneToModify {boneName=dyn_cape_01_l_MSTR}
    rotation {pitch=40}
   }
   n49 = "Transform (Modify) Bone - Bone: dyn_cape_01_r_MSTR"(ComponentPose=@n48, Alpha=@n47, Translation="0,0,0", Rotation="40,0,0", Scale="1,1,1") {
    rotationMode=BMM_Additive
    rotationSpace=BCS_BoneSpace
    translationMode=BMM_Additive
    translationSpace=BCS_BoneSpace
    alphaBoolBlend {blendOption=Linear}
    alphaScaleBiasClamp {bInterpResult=true}
    boneToModify {boneName=dyn_cape_01_r_MSTR}
    rotation {pitch=40}
   }
   n50 = Trail controller - Bone: dyn_cape_011_l!(ComponentPose=@n49, Alpha=1, RelaxationSpeedScale:float=1, FakeVelocity:Vector="0,0,0") -> (Pose) {
    bActorSpaceFakeVel=true
    bShowTrailLocation=true
    bUsePlanarLimit=true
    chainLength=9
    debugLifeTime=0.1
    maxDeltaTime=0.0333
    relaxationSpeedScale=1.2
    alphaScaleBiasClamp {bInterpResult=true, interpSpeedDecreasing=25, interpSpeedIncreasing=25}
    baseJoint {boneName=backpack_root}
    rotationLimits [
     {limitMax=<180 180 180>}
     {limitMax=<180 180 180>}
     {limitMax=<180 180 180>}
     {limitMax=<180 180 180>}
     {limitMax=<180 180 180>}
     {limitMax=<180 180 180>}
     {limitMax=<180 180 180>}
     {limitMax=<180 180 180>}
     {limitMax=<180 180 180>}
    ]
    rotationOffsets [<0 0 0>, <0 0 0>, <0 0 0>, <0 0 0>, <0 0 0>, <0 0 0>, <0 0 0>, <0 0 0>, <0 0 0>]
    trailBone {boneName=dyn_cape_011_l}
    trailRelaxationSpeed {editorCurveData {keys [<at 0 15 RCIM_Linear>,<at 0.5289 13.1838 RCIM_Linear>,<at 1 6 RCIM_Linear>]}}
   }
   n51 = Trail controller - Bone: dyn_cape_010_r!(ComponentPose=@n50, Alpha=1, RelaxationSpeedScale:float=1, FakeVelocity:Vector="0,0,0") -> (Pose) {
    bActorSpaceFakeVel=true
    bShowTrailLocation=true
    bUsePlanarLimit=true
    chainLength=9
    debugLifeTime=0.1
    maxDeltaTime=0.0333
    relaxationSpeedScale=1.2
    alphaScaleBiasClamp {bInterpResult=true, interpSpeedDecreasing=25, interpSpeedIncreasing=25}
    baseJoint {boneName=backpack_root}
    rotationLimits [
     {limitMax=<180 180 180>}
     {limitMax=<180 180 180>}
     {limitMax=<180 180 180>}
     {limitMax=<180 180 180>}
     {limitMax=<180 180 180>}
     {limitMax=<180 180 180>}
     {limitMax=<180 180 180>}
     {limitMax=<180 180 180>}
     {limitMax=<180 180 180>}
    ]
    rotationOffsets [<0 0 0>, <0 0 0>, <0 0 0>, <0 0 0>, <0 0 0>, <0 0 0>, <0 0 0>, <0 0 0>, <0 0 0>]
    trailBone {boneName=dyn_cape_010_r}
    trailRelaxationSpeed {editorCurveData {keys [<at 0 15 RCIM_Linear>,<at 0.5289 13.1838 RCIM_Linear>,<at 1 6 RCIM_Linear>]}}
   }
   Save cached pose 'trailControlCache'(Pose=Component To Local(ComponentPose=@n51)) {cacheName=trailControlCache}
  ]}
 EventGraph {nodes [Event BlueprintUpdateAnimation()]}
]
variables []
